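# Niagara effect for the JadeTowel drum kit cosmetic, surrounding the instrument with music note and dot sprites.
# Notes and Dots spawn on skeletal mesh sockets, using per-partition arrays to shape each spawn volume, and scale with clamped audio amplitude.
# Notes animate through SubUV frames and emit location events, while Notes_Glitch, lacking its own spawn module, draws brief color-stepping streaks.
# Front-end, time-of-day, platform and distance inputs tune brightness or size.
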NiagaraSystem /SparksCosmetics/Instrument/DrumKit/JadeTowel/Effects/NS_DrumKit_JadeTowel
dependencies {
 usedDataInterfaces [
  &NiagaraDataInterfaceCurve
  &NiagaraDataInterfaceArrayFloat
  &NiagaraDataInterfaceSpriteRendererInfo
  &NiagaraDataInterfaceArrayInt32
  &NiagaraDataInterfaceArrayFloat3
  &NiagaraDataInterfaceColorCurve
  &NiagaraDataInterfacePlatformSet
  &NiagaraDataInterfaceSkeletalMesh
 ]
 usedDynamicInputs []
 usedModules [
  &self:SystemSpawnScript.SystemScriptSource.SystemScriptGraph.NiagaraNodeAssignment_0.SetVariables_#1
  &self:SystemSpawnScript.SystemScriptSource.SystemScriptGraph.NiagaraNodeAssignment_1.SetVariables_#2
  &self:SystemSpawnScript.SystemScriptSource.SystemScriptGraph.NiagaraNodeAssignment_2.SetVariables_#3
  &SystemState
 ]
}
emitters {
 Dots {
  emitterUpdateScript [
   &EmitterState(Inactive Response=::NewEnumerator1, Life Cycle Mode=::NewEnumerator1, Loop Duration=1, UseLoopDelay=false)
   &SpawnRate(
    Spawn Probability=&MakeCustomFloatFromBool(Bool=&SetBoolByFloatComparison(A=$System.Music_Reactive_FX_Preview, B=0), True Float=1, False Float=1)
    SpawnRate=&Multiply_Float(A=60, B=$System.DistanceSpawn)
    Use Spawn Probability=true
   )
  ]
  particleSpawnScript [
   &PartitionParticles(Number Of Partitions=7)
   &InitializeParticle(
    Color Maximum=&MakeCustomLinearColorFromBool(Bool=&RandomBool(), True Color=<rgba 0 0.6562 0.3182 1>, False Color=<rgba 0 0.232 1 1>)
    Color Minimum=&MakeCustomLinearColorFromBool(Bool=&RandomBool(), True Color=<rgba 0 1 0.8473 1>, False Color=<rgba 0.3102 0 1 1>)
    Color Mode=::NewEnumerator2
    Lifetime Max=1.5
    Lifetime Min=0.7
    Lifetime Mode=::NewEnumerator1
    Sprite Rotation Angle=&RandomRangeFloat(Minimum=-0.15, Maximum=0.15)
    Sprite Rotation Mode=::NewEnumerator1
    Sprite Size Mode=::NewEnumerator1
    Uniform Sprite Size Max=5
    Uniform Sprite Size Min=3
    UsePositionOffset=false
    Write Material Random=true
   )
   /SetVariables(Particles.VisibilityTag=&RandomRangeInt(Minimum=0, Maximum=2)) {isSetParameters=true}
   &SkeletalMeshLocation(
    Bone / Socket Index=&SelectIntFromArray(Direct Array Index=$Particles.PartitionParticles.Partition, Int Selection Array=DataInterface() {propertyValues {IntData [0,1,2,3,4,5,6]}})
    MaskSpawnGroup=false
    Mesh Sampling Type=::NewEnumerator5
    Skeletal Mesh=$System.Skeletal Mesh
    Socket Sampling Mode=::NewEnumerator5
   )
   &ShapeLocation(
    Handle Radius:NiagaraFloat=&SelectFloatFromArray(
     Direct Array Index=$Particles.PartitionParticles.Partition
     Float Selection Array=DataInterface() {propertyValues {FloatData [3,3,3,3,17,3,3]}}
    )
    Large Radius=&SelectFloatFromArray(
     Direct Array Index=$Particles.PartitionParticles.Partition
     Float Selection Array=DataInterface() {propertyValues {FloatData [6,6,6,6,17,25,25]}}
    )
    MaskSpawnGroup=true
    Non Uniform Scale=&SelectVectorFromArray(
     Direct Array Index=$Particles.PartitionParticles.Partition
     Vector Selection Array=DataInterface() {propertyValues {FloatData [<1 1 11>, <1 1 15>, <1 1 13>, <1 1 9>, <1 1 0.3>, <1 1 7>, <1 1 4>]}}
     Random Seed=0
    )
    Offset=&SelectVectorFromArray(
     Direct Array Index=$Particles.PartitionParticles.Partition
     Vector Selection Array=DataInterface() {propertyValues {FloatData [<0 0 0>, <0 0 2>, <0 0 -5>, <0 0 0>, <0 0 0>, <0 0 0>, <0 0 0>]}}
     Random Seed=0
    )
    Offset Mode=::NewEnumerator0
    Override Seed=false
    Rotation Angle Type:ENiagara_AngleInput=::NewEnumerator0
    Rotation Axis=&SelectVectorFromArray(
     Direct Array Index=$Particles.PartitionParticles.Partition
     Vector Selection Array=DataInterface() {propertyValues {FloatData [<0 0 1>, <0 0 1>, <0 0 1>, <0 0 1>, <0 1 0>, <0 0 1>, <0 0 1>]}}
     Random Seed=0
    )
    Rotation Mode=::NewEnumerator1
    Shape Primitive=::NewEnumerator3
    UseOwnerScale=true
   ) {module=ShapeLocation002}
   &CameraOffset(Camera Offset Amount=10)
  ]
  particleUpdateScript [
   &ParticleState()
   &SpriteRotationRate(Rotation Rate=&Multiply_Float(
     A=&RandomRangeFloat(Minimum=-10, Maximum=10)
     B=&FloatFromCurve(FloatCurve=DataInterface() {propertyValues {Curve {keys [<at 0 1>,<at 1 0>]}}}, Scale Curve=10)
    ))
   &ScaleSpriteSize(
    Uniform Curve Scale=&RandomRangeFloat(Minimum=0.6, Maximum=1)
    Uniform Curve Sprite Scale=DataInterface() {propertyValues {Curve {keys [{tangentMode=RCTM_User, time=0, value=0.4}, <at 0.1 1.2>, <at 0.2 1>, {tangentMode=RCTM_User, time=1, value=1}]}}}
   )
   &ScaleSpriteSize(
    Uniform Curve Scale=&Multiply_Float(
     A=&RandomRangeFloat(Minimum=0.7, Maximum=1)
     B=&Multiply_Float(
      A=&MakeCustomFloatFromBool(
       Bool=&SetBoolByFloatComparison(A=$System.Music_Reactive_FX_Preview, B=0)
       True Float=&MakeCustomFloatFromBool(Bool=$User.bIsFrontEndPreview, True Float=1, False Float=1)
       False Float=&MakeCustomFloatFromBool(Bool=$User.bIsFrontEndPreview, True Float=3, False Float=3)
      )
      B=&ClampFloat(Float=$NPC.FortniteNPC.EventAudio_Amplitude, Min=0.3, Max=0.8)
     )
    )
    Uniform Curve Sprite Scale=DataInterface() {propertyValues {Curve {keys [{tangentMode=RCTM_User, time=0, value=0.3}, <at 0.1 1.2>, <at 0.2 1>, {tangentMode=RCTM_User, time=1, value=1}]}}}
   ) {module=ScaleSpriteSize001}
   /SetVariables(Particles.Size002:Vector2f=&Multiply_Vector2D(
     A=$Particles.SpriteSize
     B=&Vector2DFromFloat(Value=&RandomRangeFloat(Minimum=0.5, Maximum=0.8))
    )) {isSetParameters=true}
   &ScaleColor(
    Scale Alpha=&FloatFromCurve(
     FloatCurve=DataInterface() {propertyValues {Curve {keys [<at 0.5 1>,<at 1 0>]}}}
     Scale Curve=&MakeCustomFloatFromBool(
      Bool=&SetBoolByFloatComparison(A=$System.Music_Reactive_FX_Preview, B=0)
      True Float=&Multiply_Float(
       A=&FN_FrontEndFloatModifier(Gameplay=2, Front End=0.4)
       B=&FNTimeOfDayScalar(Night Scale=0.8, Morning Scale=1, Evening Scale=1)
      )
      False Float=1
     )
    )
    ScaleA=true
    ScaleRGB=true
   )
   /SetVariables(Particles.Color_glow:LinearColor=&MakeLinearColorFromVectorAndFloat(
     "Vector (RGB)"=&MakeVectorFromLinearColorRGB(Color=<rgba 1 1 1 1>)
     "Float (Alpha)"=&FloatFromCurve(
      FloatCurve=DataInterface() {propertyValues {Curve {keys [<at 0.5 1>,<at 1 0>]}}}
      Scale Curve=&MakeCustomFloatFromBool(
       Bool=&SetBoolByFloatComparison(A=$System.Music_Reactive_FX_Preview, B=0)
       True Float=&Multiply_Float(
        A=&FN_FrontEndFloatModifier(Gameplay=2, Front End=0.3)
        B=&FNTimeOfDayScalar(Night Scale=0.1, Morning Scale=1, Evening Scale=1)
       )
       False Float=1
      )
     )
    )) {isSetParameters=true}
   &SolveForcesAndVelocity(Clamp Velocity=false, Limit Acceleration=false)
   /SetVariables(Particles.Position=&Lerp_Vector(
     A=Linked() {linkedVariable {name=Particles.Initial.Position, type=NiagaraPosition}}
     B=&AddVectorToPosition(Position=$Particles.Initial.Position, Vector=&RandomRangeVector(Minimum=<0 0 -1>, Maximum=<1 1 -6>))
     Alpha=&FloatFromCurve(FloatCurve=DataInterface() {propertyValues {Curve {keys [<at 0 0>, {leaveTangent=2.691, time=1, value=1}]}}})
    )) {isSetParameters=true}
  ]
  rendererClasses [&NiagaraSpriteRendererProperties]
  renderers [
   {rendererClass=&NiagaraSpriteRendererProperties}
   {rendererClass=&NiagaraSpriteRendererProperties, rendererIndex=1}
  ]
 }
 Notes {
  emitterUpdateScript [
   &EmitterState(Inactive Response=::NewEnumerator1, Life Cycle Mode=::NewEnumerator1, Loop Duration=1, UseLoopDelay=false)
   &SpawnRate(
    Spawn Probability=&MakeCustomFloatFromBool(Bool=&SetBoolByFloatComparison(A=$System.Music_Reactive_FX_Preview, B=1), True Float=1, False Float=1)
    SpawnRate=&MakeCustomFloatFromBool(Bool=true, True Float=9, False Float=&Multiply_Float(A=2, B=$System.DistanceSpawn))
    Use Spawn Probability=true
   )
  ]
  particleSpawnScript [
   &PartitionParticles(Number Of Partitions=5)
   &InitializeParticle(
    Color=<rgba 1 1 1 1.5>
    Lifetime Max=1
    Lifetime Mode=::NewEnumerator1
    Sprite Rotation Angle=&RandomRangeFloat(Minimum=-0.15, Maximum=0.15)
    Sprite Rotation Mode=::NewEnumerator1
    Sprite Size=&Multiply_Vector2D_ByFloat(
     Float=&SelectFloatFromArray(Direct Array Index=$Particles.PartitionParticles.Partition, Float Selection Array=DataInterface() {propertyValues {FloatData [1,1,1,1,1]}})
     Vector2D=&Multiply_Vector2D_ByFloat(
      Float=&MakeCustomFloatFromBool(Bool=$User.bIsFrontEndPreview, True Float=1.25, False Float=1)
      Vector2D=&RandomRangeVector2D(Minimum=<10 16>, Maximum=<12 18>)
     )
    )
    Sprite Size Mode=::NewEnumerator4
    UsePositionOffset=false
    Write Material Random=true
   )
   &SubUVAnimation(
    End Frame Range Override=6
    "Sprite Renderer "=DataInterface() {propertyValues {SpriteRenderer=&self:Notes_0.NiagaraSpriteRendererProperties_0}}
    SubUV Animation Mode:ENiagara_SubUVLookupModeV2=::NewEnumerator2
    UseEndFrame=true
    UseStartFrame=true
   )
   &SkeletalMeshLocation(
    Bone / Socket Index=&SelectIntFromArray(Direct Array Index=$Particles.PartitionParticles.Partition, Int Selection Array=DataInterface() {propertyValues {IntData [0,1,2,3,4]}})
    MaskSpawnGroup=false
    Mesh Sampling Type=::NewEnumerator5
    Skeletal Mesh=$System.Skeletal Mesh
    Socket Sampling Mode=::NewEnumerator5
   )
   &ShapeLocation(
    Handle Radius:NiagaraFloat=&SelectFloatFromArray(Direct Array Index=$Particles.PartitionParticles.Partition, Float Selection Array=DataInterface() {propertyValues {FloatData [3,3,3,3,17]}})
    Large Radius=&SelectFloatFromArray(Direct Array Index=$Particles.PartitionParticles.Partition, Float Selection Array=DataInterface() {propertyValues {FloatData [6,6,6,6,17]}})
    MaskSpawnGroup=true
    Non Uniform Scale=&SelectVectorFromArray(
     Direct Array Index=$Particles.PartitionParticles.Partition
     Vector Selection Array=DataInterface() {propertyValues {FloatData [<1 1 11>, <1 1 15>, <1 1 13>, <1 1 9>, <1 1 0.3>]}}
    )
    Offset=&SelectVectorFromArray(
     Direct Array Index=$Particles.PartitionParticles.Partition
     Vector Selection Array=DataInterface() {propertyValues {FloatData [<0 0 0>, <0 0 2>, <0 0 -5>, <0 0 0>, <0 0 0>]}}
    )
    Offset Mode=::NewEnumerator0
    Override Seed=false
    Rotation Angle Type:ENiagara_AngleInput=::NewEnumerator0
    Rotation Axis=&SelectVectorFromArray(
     Direct Array Index=$Particles.PartitionParticles.Partition
     Vector Selection Array=DataInterface() {propertyValues {FloatData [<0 0 1>, <0 0 1>, <0 0 1>, <0 0 1>, <0 1 0>]}}
    )
    Rotation Mode=::NewEnumerator1
    Shape Primitive=::NewEnumerator3
    UseOwnerScale=true
   )
  ]
  particleUpdateScript [
   &ParticleState()
   &SpriteRotationRate(Rotation Rate=&Multiply_Float(
     A=&RandomRangeFloat(Minimum=-10, Maximum=10)
     B=&FloatFromCurve(FloatCurve=DataInterface() {propertyValues {Curve {keys [<at 0 1>,<at 1 0>]}}}, Scale Curve=5)
    ))
   &ScaleSpriteSize(
    Uniform Curve Scale=&Multiply_Float(
     A=&RandomRangeFloat(Minimum=0.7, Maximum=1)
     B=&Multiply_Float(
      A=&MakeCustomFloatFromBool(
       Bool=&SetBoolByFloatComparison(A=$System.Music_Reactive_FX_Preview, B=0)
       True Float=&MakeCustomFloatFromBool(Bool=$User.bIsFrontEndPreview, True Float=1.5, False Float=1.5)
       False Float=&MakeCustomFloatFromBool(Bool=$User.bIsFrontEndPreview, True Float=3, False Float=3)
      )
      B=&ClampFloat(Float=$NPC.FortniteNPC.EventAudio_Amplitude, Min=0.3, Max=0.8)
     )
    )
    Uniform Curve Sprite Scale=DataInterface() {propertyValues {Curve {keys [{tangentMode=RCTM_User, time=0, value=0.3}, <at 0.1 1.2>, <at 0.2 1>, {tangentMode=RCTM_User, time=1, value=1}]}}}
   )
   &ScaleSpriteSize(
    Uniform Curve Index=&FN_LOD_Distance_Remap(Start Distance=0, End Distance=1000, Minimum Output=0, Maximum Output=1, Clamp Output=true)
    Uniform Curve Sprite Scale=DataInterface() {propertyValues {Curve {keys [{arriveTangent=1.5779, time=0, value=0.6}, <at 1 1>]}}}
   ) {module=ScaleSpriteSize002}
   /SetVariables(Particles.Size002:Vector2f=&Multiply_Vector2D(A=$Particles.SpriteSize, B=&Vector2DFromFloat(Value=1.1))) {isSetParameters=true}
   &DynamicMaterialParameters(
    Index 0 Param 1=&FloatFromCurve(FloatCurve=DataInterface() {propertyValues {Curve {keys [<at 0 -1.5>,<at 0.5 -1.5>,<at 1 1>]}}})
    Index 0 Param 2=0.2
    Module.Index0Param0Write=true
    Module.Index0Param1Write=true
    Module.Index0Param2Write=true
    Module.Index0Param3Write=true
   )
   &ScaleColor(
    Scale Alpha=&MakeCustomFloatFromBool(
     Bool=&SetBoolByFloatComparison(A=$System.Music_Reactive_FX_Preview, B=0)
     True Float=&Multiply_Float(
      A=&Lerp_Float(
       A=&FN_FrontEndFloatModifier(
        Gameplay=&FN_PickFloatDependOnPlatform(TargetPlatform=DataInterface() {propertyValues {Platforms {qualityLevelMask=-1}}}, TargetPlatformActive=1, TargetPlatformNoActive=0.8)
        Front End=0.6
       )
       B=&FN_FrontEndFloatModifier(
        Gameplay=&FN_PickFloatDependOnPlatform(TargetPlatformActive=3, TargetPlatformNoActive=1.5)
        Front End=1
       )
       Alpha=$System.Music_Reactive_FX_Preview
      )
      B=&FNTimeOfDayScalar(Night Scale=0.3, Morning Scale=1, Evening Scale=1)
     )
     False Float=1
    )
    Scale RGB=&MakeVectorFromLinearColorRGB(Color=&ColorFromCurve(ColorCurve=DataInterface() {propertyValues {
        AlphaCurve {keys []}
        BlueCurve {keys [
          <at 0.0474 2.3109 RCIM_Cubic>
          <at 0.1465 0.0526>
          <at 0.3482 2.3109>
          {arriveTangent=-2.6251, interpMode=RCIM_Cubic, leaveTangent=-2.6251, time=0.7543, value=0.875}
          <at 0.9128 0.8288>
         ]}
        GreenCurve {keys [
          <at 0.0474 3 RCIM_Cubic>
          <at 0.1465 2>
          <at 0.3482 3>
          {arriveTangent=-4.5294, interpMode=RCIM_Cubic, leaveTangent=-4.5294, time=0.7543, value=0.7282}
          <at 0.9128 0.4427>
         ]}
        RedCurve {keys [
          <at 0.0474 0 RCIM_Cubic>
          <at 0.1465 0.1374>
          <at 0.3482 0>
          {arriveTangent=1.7712, interpMode=RCIM_Cubic, leaveTangent=1.7712, time=0.7543, value=0}
          <at 0.9128 1>
         ]}
       }}))
    ScaleA=true
    ScaleRGB=true
   )
   /SetVariables(Particles.Color_glow:LinearColor=&MakeLinearColorFromVectorAndFloat(
     "Vector (RGB)"=&MakeVectorFromLinearColorRGB(Color=<rgba 0 1 0.6636 1>)
     "Float (Alpha)"=&MakeCustomFloatFromBool(
      Bool=&SetBoolByFloatComparison(A=$System.Music_Reactive_FX_Preview, B=0)
      True Float=&Multiply_Float(
       A=&FN_FrontEndFloatModifier(Gameplay=0.3, Front End=0.15)
       B=&FNTimeOfDayScalar(Night Scale=0.8, Morning Scale=1, Evening Scale=1)
      )
      False Float=0
     )
    )) {isSetParameters=true}
   /SetVariables(Particles.Color_glow2:LinearColor=&MakeLinearColorFromVectorAndFloat(
     "Vector (RGB)"=&Multiply_Vector(
      A=&MakeVectorFromLinearColorRGB(Color=<rgba 1 0.0197 0.7901 1>)
      B=&VectorFromFloat(Value=&FN_PickFloatDependOnPlatform(TargetPlatformActive=1, TargetPlatformNoActive=0.5))
     )
     "Float (Alpha)"=&MakeCustomFloatFromBool(
      Bool=&SetBoolByFloatComparison(A=$System.Music_Reactive_FX_Preview, B=0)
      True Float=&Multiply_Float(
       A=&Lerp_Float(
        A=&FN_FrontEndFloatModifier(Gameplay=0.3, Front End=0.1)
        B=&FN_FrontEndFloatModifier(Gameplay=0.3, Front End=0.2)
        Alpha=$System.Music_Reactive_FX_Preview
       )
       B=&FNTimeOfDayScalar(Night Scale=0.8, Morning Scale=1, Evening Scale=1)
      )
      False Float=0
     )
    )) {isSetParameters=true}
   &SkeletalMeshLocation(
    Bone / Socket Index=&SelectIntFromArray(Direct Array Index=$Particles.PartitionParticles.Partition, Int Selection Array=DataInterface() {propertyValues {IntData [0,1,2,3,4]}})
    MaskSpawnGroup=false
    Mesh Sampling Type=::NewEnumerator5
    Skeletal Mesh=$System.Skeletal Mesh
    Socket Sampling Mode=::NewEnumerator5
   ) {module=SkeletalMeshLocation001}
   &ShapeLocation(
    Handle Radius:NiagaraFloat=&SelectFloatFromArray(Direct Array Index=$Particles.PartitionParticles.Partition, Float Selection Array=DataInterface() {propertyValues {FloatData [3,3,3,3,17]}})
    Large Radius=&SelectFloatFromArray(Direct Array Index=$Particles.PartitionParticles.Partition, Float Selection Array=DataInterface() {propertyValues {FloatData [6,6,6,6,17]}})
    MaskSpawnGroup=true
    Non Uniform Scale=&SelectVectorFromArray(
     Direct Array Index=$Particles.PartitionParticles.Partition
     Vector Selection Array=DataInterface() {propertyValues {FloatData [<1 1 11>, <1 1 15>, <1 1 13>, <1 1 9>, <1 1 0.3>]}}
     Random Seed=0
    )
    Offset=&SelectVectorFromArray(
     Direct Array Index=$Particles.PartitionParticles.Partition
     Vector Selection Array=DataInterface() {propertyValues {FloatData [<0 0 0>, <0 0 2>, <0 0 -5>, <0 0 0>, <0 0 0>]}}
     Random Seed=0
    )
    Offset Mode=::NewEnumerator0
    Override Seed=false
    Rotation Angle Type:ENiagara_AngleInput=::NewEnumerator0
    Rotation Axis=&SelectVectorFromArray(
     Direct Array Index=$Particles.PartitionParticles.Partition
     Vector Selection Array=DataInterface() {propertyValues {FloatData [<0 0 1>, <0 0 1>, <0 0 1>, <0 0 1>, <0 1 0>]}}
     Random Seed=0
    )
    Rotation Mode=::NewEnumerator1
    Shape Origin=$StackContext.Position
    Shape Primitive=::NewEnumerator3
    UseOwnerScale=true
   ) {module=ShapeLocation001}
   &SolveForcesAndVelocity(Clamp Velocity=false, Limit Acceleration=false)
   &GenerateLocationEvent(
    Event Probability=1
    Event Probability Evaluation Type:ENiagaraRandomnessEvaluation=::NewEnumerator0
    Event Send Rate:NiagaraFloat=&Multiply_Float(A=3, B=$System.DistanceSpawn)
    Use Event Probability=true
    UseEventDelay=false
   )
  ]
  rendererClasses [&NiagaraSpriteRendererProperties]
  renderers [
   {rendererClass=&NiagaraSpriteRendererProperties}
   {rendererClass=&NiagaraSpriteRendererProperties, rendererIndex=1}
   {rendererClass=&NiagaraSpriteRendererProperties, rendererIndex=2}
  ]
 }
 Notes_Glitch {
  emitterUpdateScript [&EmitterState(Inactive Response=::NewEnumerator1, Life Cycle Mode=::NewEnumerator1, Loop Duration=1, UseLoopDelay=false)]
  particleSpawnScript [&InitializeParticle(
    Lifetime Max=0.35
    Lifetime Min=0.15
    Lifetime Mode=::NewEnumerator1
    Mesh Scale Max=<0.5 0.5 10>
    Mesh Scale Min=<0.1 0.1 5>
    Mesh Scale Mode=::NewEnumerator2
    Sprite Size Max=<1.5 16>
    Sprite Size Min=<0.5 10>
    Sprite Size Mode=::NewEnumerator2
    UsePositionOffset=false
    Write Material Random=true
   )]
  particleUpdateScript [
   &ParticleState()
   &ScaleMeshSize(Scale Factor=&Multiply_VectorByFloat(
     Vector=&MakeVector(
      X=1
      Y=1
      Z=&FloatFromCurve(FloatCurve=DataInterface() {propertyValues {Curve {keys [<at 0 0>,<at 0.2 0.95>,<at 1 1>]}}})
     )
     Float=&Multiply_Float(
      A=0.1
      B=&FloatFromCurve(
       FloatCurve=DataInterface() {propertyValues {Curve {keys [
           {arriveTangent=4.1785, time=0, value=0}
           <at 0.5 1.1>
           {arriveTangent=-0.0405, interpMode=RCIM_Cubic, leaveTangent=-0.0405, tangentMode=RCTM_User, time=1, value=1.5}
          ]}}}
       Scale Curve=&Multiply_Float(
        A=1
        B=&ClampFloat(
         Float=&FN_LOD_Distance_Remap(Start Distance=200, End Distance=400, Minimum Output=0.5, Maximum Output=1)
         Min=0.5
        )
       )
      )
     )
    ))
   &SolveForcesAndVelocity(Clamp Velocity=false, Limit Acceleration=false)
   &ScaleColor(
    Scale Alpha=&Multiply_Float(
     A=&FloatFromCurve(
      FloatCurve=DataInterface() {propertyValues {Curve {keys [<at 0 1>,<at 1 0>]}}}
      Scale Curve=&RandomRangeFloat(Minimum=0.4, Maximum=1)
     )
     B=3
    )
    Scale RGB=&Multiply_VectorByFloat(
     Vector=&MakeVectorFromLinearColorRGB(Color=&ColorFromCurve(ColorCurve=DataInterface() {propertyValues {
         AlphaCurve {keys []}
         BlueCurve {keys [<at 0.1306 0.8473 RCIM_Constant>,<at 0.1859 0.4496 RCIM_Constant>,<at 0.3396 0.8473 RCIM_Constant>,<at 0.471 1 RCIM_Constant>,<at 0.5576 1 RCIM_Constant>,<at 0.68 0.3111 RCIM_Constant>,<at 0.9084 0.8473>]}
         GreenCurve {keys [<at 0.1306 1 RCIM_Constant>,<at 0.1859 1 RCIM_Constant>,<at 0.3396 1 RCIM_Constant>,<at 0.471 0 RCIM_Constant>,<at 0.5576 0.4231 RCIM_Constant>,<at 0.68 1 RCIM_Constant>,<at 0.9084 1>]}
         RedCurve {keys [<at 0.1306 0 RCIM_Constant>,<at 0.1859 0 RCIM_Constant>,<at 0.3396 0 RCIM_Constant>,<at 0.471 0.9661 RCIM_Constant>,<at 0.5576 0 RCIM_Constant>,<at 0.68 0 RCIM_Constant>,<at 0.9084 0>]}
        }}))
     Float=&FloatFromCurve(
      FloatCurve=DataInterface() {propertyValues {Curve {keys [
          {arriveTangent=-9.3569, interpMode=RCIM_Cubic, leaveTangent=-9.3569, tangentMode=RCTM_User, time=0, value=2}
          {arriveTangent=-2.1612, interpMode=RCIM_Cubic, leaveTangent=-2.1612, tangentMode=RCTM_User, time=0.1, value=1}
          <at 1 0.3>
         ]}}}
      CurveIndex=$Emitter.Age
      Scale Curve=1.5
     )
    )
    ScaleA=true
    ScaleRGB=true
   )
   /SetVariables(Particles.Color_glow:LinearColor=&MakeLinearColorFromVectorAndFloat(
     "Vector (RGB)"=&MakeVectorFromLinearColorRGB(Color=&ColorFromCurve(ColorCurve=DataInterface() {propertyValues {
         AlphaCurve {keys []}
         BlueCurve {keys [
           {interpMode=RCIM_Cubic, tangentMode=RCTM_User, time=0.1306, value=0.8473}
           {interpMode=RCIM_Cubic, tangentMode=RCTM_User, time=0.3519, value=0.4336}
           {interpMode=RCIM_Cubic, tangentMode=RCTM_User, time=0.474, value=1}
           {interpMode=RCIM_Cubic, tangentMode=RCTM_User, time=0.5662, value=1}
           {interpMode=RCIM_Cubic, tangentMode=RCTM_User, time=0.7015, value=0.8473}
          ]}
         GreenCurve {keys [
           {interpMode=RCIM_Cubic, tangentMode=RCTM_User, time=0.1306, value=1}
           {interpMode=RCIM_Cubic, tangentMode=RCTM_User, time=0.3519, value=1}
           {interpMode=RCIM_Cubic, tangentMode=RCTM_User, time=0.474, value=0}
           {arriveTangent=3.4579, interpMode=RCIM_Cubic, leaveTangent=3.4579, tangentMode=RCTM_User, time=0.5662, value=0.4231}
           {interpMode=RCIM_Cubic, tangentMode=RCTM_User, time=0.7015, value=1}
          ]}
         RedCurve {keys [
           {interpMode=RCIM_Cubic, tangentMode=RCTM_User, time=0.1306, value=0}
           {interpMode=RCIM_Cubic, tangentMode=RCTM_User, time=0.3519, value=0}
           {interpMode=RCIM_Cubic, tangentMode=RCTM_User, time=0.474, value=0.9661}
           {interpMode=RCIM_Cubic, tangentMode=RCTM_User, time=0.5662, value=0}
           {interpMode=RCIM_Cubic, tangentMode=RCTM_User, time=0.7015, value=0}
          ]}
        }}))
     "Float (Alpha)"=&FloatFromCurve(
      FloatCurve=DataInterface() {propertyValues {Curve {keys [<at 0.5 1>,<at 1 0>]}}}
      Scale Curve=&MakeCustomFloatFromBool(
       Bool=&SetBoolByFloatComparison(A=$System.Music_Reactive_FX_Preview, B=0)
       True Float=&Multiply_Float(
        A=&RandomRangeFloat(Minimum=1, Maximum=2)
        B=&Lerp_Float(
         A=&FN_FrontEndFloatModifier(Gameplay=1, Front End=0.7)
         B=&FN_FrontEndFloatModifier(Gameplay=2, Front End=1.4)
         Alpha=$System.Music_Reactive_FX_Preview
        )
       )
       False Float=1
      )
     )
    )) {isSetParameters=true}
   &CameraOffset(Camera Offset Amount=10)
  ]
  rendererClasses [&NiagaraSpriteRendererProperties]
  renderers [{rendererClass=&NiagaraSpriteRendererProperties}]
 }
}
summary {
 systemName=NS_DrumKit_JadeTowel
 emitters [
  {emitterName=Notes, rendererClasses [&NiagaraSpriteRendererProperties]}
  {emitterName=Notes_Glitch, rendererClasses [&NiagaraSpriteRendererProperties]}
  {emitterName=Dots, rendererClasses [&NiagaraSpriteRendererProperties]}
 ]
 userVariables [
  User.bIsFrontEnd {type=NiagaraBool}
  User.bIsFrontEndPreview {type=NiagaraBool}
  User.IsFrontEnd {type=NiagaraBool}
 ]
}
